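# Player AnimBlueprint for VoidSlide seating, supplying Item Override layers that pose the character in one of four seats.
# FullBody and FinalPose layers duplicate Chooser-driven Intro, Sitting and Idle Break state machines, with seat swaps re-entering Sitting.
# Both apply spawn-prop additives and per-seat root offsets selected by SeatIndex.
# FinalPose switches to its input pose after a short timer, leaving FullBody to carry the sit.
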AnimBlueprint /WildEstate/Gameplay/VoidSlide/Sequence/VoidSlide_Player_AnimBP
target_skeleton=&Fortnite_M_Avg_Player_Skeleton
graphs [
 ANF_Disable Sit Additive {nodes [
   ANF_Disable Sit Additive!() -> (then, Context:"Anim Update Context Structure (by ref)", Node:"Anim Node Reference Structure (by ref)") {is=title K2Node_FunctionEntry}
   > Set Sit Additive(Sit Additive:bool=false) -> (Output_Get:bool)
   > Return Node()
  ]}
 ANF_DisableIdleBreak {nodes [
   ANF_DisableIdleBreak!() -> (then, Context:"Anim Update Context Structure (by ref)", Node:"Anim Node Reference Structure (by ref)") {is=title K2Node_FunctionEntry}
   > Set bTriggerIdleBreak(bTriggerIdleBreak:bool=false) -> (Output_Get:bool)
   > Set TriggerTimeRemaining(TriggerTimeRemaining:double=5) -> (Output_Get:double)
   > Return Node()
  ]}
 ANF_Enable Sit Additive {nodes [
   ANF_Enable Sit Additive!() -> (then, Context:"Anim Update Context Structure (by ref)", Node:"Anim Node Reference Structure (by ref)") {is=title K2Node_FunctionEntry}
   > Set Sit Additive(Sit Additive=true)
   > Return Node()
  ]}
 AnimGraph {nodes [Output Pose()]}
 EventGraph {nodes [
   n1 = AnimNotify_Reset Idle Break Timer() {is=title K2Node_Event}
   Set TriggerTimeRemaining(
    execute=[@n1.then, @n2.then]
    TriggerTimeRemaining=RandomFloatInRange(Min=Get Min Idle Break Time() -> (Min Idle Break Time:double), Max=Get Max Idle Break Time() -> (Max Idle Break Time:double))
   )
   > Set bTriggerIdleBreak(bTriggerIdleBreak=false)
   AnimNotify_Reenter Sitting Event() {is=title K2Node_Event}
   > n2 = Set Sit Additive(Sit Additive=true)
  ]}
 ItemAimOffset {nodes [
   Input Pose!() -> (Pose, AimOffsetAlpha:double, Yaw:double, Pitch:double) {name=InPose_AimOffset}
   Input Pose!() -> (Pose) {name=InPose_PreAimOffsets}
   Output Pose() {layerGroup=Item Override}
  ]}
 ItemFallAdditive {nodes [
   Input Pose() {name=InPoseItemFallAdditive}
   Output Pose() {layerGroup=Item Override}
  ]}
 ItemFallLandAdditive {nodes [
   Input Pose() {name=InPoseItemFallLandAdditive}
   Output Pose() {layerGroup=Item Override}
  ]}
 ItemFallLevel2Override {nodes [
   Input Pose() {name=InPose_FallLevel2}
   Output Pose() {layerGroup=Item Override}
  ]}
 ItemFinalPoseOverride {nodes [
   AnimGraphNode_StateMachine_2 = Final Pose Blend Out SM() {is=title AnimGraphNode_StateMachine}
   Output Pose(Result=@AnimGraphNode_StateMachine_2) {layerGroup=Item Override}
   AnimGraphNode_StateMachine_0 = Sitting FinalPose() {is=title AnimGraphNode_StateMachine}
   n1 = Apply Additive!(
    Base=@AnimGraphNode_StateMachine_0
    Additive="Sequence Player (additive) 'WildEstate_VoidSlide_SpawnProp1'"() {sequence=&/WildEstate_VoidSlide_SpawnProp1}
    bAlphaBoolEnabled:bool=Get bIsSeat01SpawnProp() -> (bIsSeat01SpawnProp:bool)
   ) -> (Pose) {alphaInputType=Bool}
   n2 = "Sequence Player (additive) 'WildEstate_VoidSlide_SpawnProp3'"() {sequence=&/WildEstate_VoidSlide_SpawnProp3}
   n3 = Get bIsSeat03SpawnProp() -> (bIsSeat03SpawnProp:bool)
   Save cached pose 'FinalPose PreOffsets'(Pose=Apply Additive(
     Base=Apply Additive(Base=@n1, Additive=@n2, bAlphaBoolEnabled=@n3) {alphaInputType=Bool}
     Additive="Sequence Player (additive) 'WildEstate_VoidSlide_SpawnProp4'"() {sequence=&/WildEstate_VoidSlide_SpawnProp4}
     bAlphaBoolEnabled=Get bIsSeat04SpawnProp() -> (bIsSeat04SpawnProp:bool)
    ) {alphaInputType=Bool}) {cacheName=FinalPose PreOffsets}
   Save cached pose 'Final InputPose'(Pose=Input Pose() {name=InFinalPose}) {cacheName=Final InputPose}
   n4 = Local To Component(LocalPose=Use cached pose 'FinalPose PreOffsets'())
   n5 = Local To Component(LocalPose=Use cached pose 'FinalPose PreOffsets'())
   n6 = Local To Component(LocalPose=Use cached pose 'FinalPose PreOffsets'())
   n7 = Local To Component(LocalPose=Use cached pose 'FinalPose PreOffsets'())
   n8 = Blend Poses by int!(
    BlendPose_0:Pose Link Structure=Component To Local(ComponentPose="Transform (Modify) Bone - Bone: Root"!(ComponentPose=@n4, bAlphaBoolEnabled:bool=True, Translation="7,-28,0") -> (Pose) {alphaInputType=Bool, translationMode=BMM_Additive, boneToModify {boneName=Root}, translation=<7 -28>})
    BlendPose_1:Pose Link Structure=Component To Local(ComponentPose="Transform (Modify) Bone - Bone: Root"(ComponentPose=@n5, bAlphaBoolEnabled=True, Translation="5,-28,0") {alphaInputType=Bool, translationMode=BMM_Additive, boneToModify {boneName=Root}, translation=<5 -28>})
    BlendPose_2:Pose Link Structure=Component To Local(ComponentPose="Transform (Modify) Bone - Bone: Root"(ComponentPose=@n6, bAlphaBoolEnabled=True, Translation="0,-28,0") {alphaInputType=Bool, translationMode=BMM_Additive, boneToModify {boneName=Root}, translation {y=-28}})
    BlendPose_3:Pose Link Structure=Component To Local(ComponentPose="Transform (Modify) Bone - Bone: Root"(ComponentPose=@n7, bAlphaBoolEnabled=True, Translation="-8,-37,0") {alphaInputType=Bool, translationMode=BMM_Additive, boneToModify {boneName=Root}, translation=<-8 -37>})
    BlendTime_0:float=0
    BlendTime_1:float=0
    BlendTime_2:float=0
    BlendTime_3:float=0
    ActiveChildIndex:int=Get SeatIndex() -> (SeatIndex:int=0)
   ) -> (Pose:Pose Link Structure) {
    is=title AnimGraphNode_BlendListByInt
    blendPose [{linkId=-1, sourceLinkId=-1}, {linkId=-1, sourceLinkId=-1}, {linkId=-1, sourceLinkId=-1}, {linkId=-1, sourceLinkId=-1}]
    blendTime [0.1,0.1,0.1,0.1]
   }
   Save cached pose 'FinalPose PostOffsets'(Pose=@n8) {cacheName=FinalPose PostOffsets}
  ]}
 Sitting FinalPose {
  entry=Intro
  path=ItemFinalPoseOverride.AnimGraphNode_StateMachine_0.Sitting FinalPose
  ids=id
  nodes [
   Re-Enter Sitting() {is=title AnimStateAliasNode, also {aliases [Sit Idle Break, Sitting]}}
   AnimStateNode_0 = Intro() {is=title AnimStateNode}
   AnimStateNode_1 = Sit Idle Break() {is=title AnimStateNode}
   AnimStateNode_5 = Sitting() {is=title AnimStateNode, bAlwaysResetOnEntry=true}
   AnimStateTransitionNode_0 = AnimStateTransitionNode() {
    bAllowInertializationForSelfTransitions=true
    bAutomaticRuleBasedOnSequencePlayerInState=true
    transitionStart {guid=#1, notifyName=superbunny}
   }
   AnimStateTransitionNode_1 = AnimStateTransitionNode() {bAllowInertializationForSelfTransitions=true, crossfadeDuration=0.7}
   AnimStateTransitionNode_2 = AnimStateTransitionNode() {
    bAllowInertializationForSelfTransitions=true
    bAutomaticRuleBasedOnSequencePlayerInState=true
    crossfadeDuration=0.5
    transitionStart {guid=#2, notifyName=Reset Idle Break Timer}
   }
   AnimStateTransitionNode_9 = AnimStateTransitionNode() {
    bAllowInertializationForSelfTransitions=true
    crossfadeDuration=0
    transitionStart {guid=#3, notifyName=Reenter Sitting Event}
   }
  ]
  links [
   {
    dst=Sitting
    rule=ItemFinalPoseOverride.AnimGraphNode_StateMachine_0.Sitting FinalPose.AnimStateTransitionNode_0.Transition
    src=Intro
    via=AnimStateTransitionNode_0
   }
   {
    dst=Sitting
    rule=ItemFinalPoseOverride.AnimGraphNode_StateMachine_0.Sitting FinalPose.AnimStateTransitionNode_9.Transition
    src=Re-Enter Sitting
    via=AnimStateTransitionNode_9
   }
   {
    dst=Sitting
    rule=ItemFinalPoseOverride.AnimGraphNode_StateMachine_0.Sitting FinalPose.AnimStateTransitionNode_2.Transition
    src=Sit Idle Break
    via=AnimStateTransitionNode_2
   }
   {
    dst=Sit Idle Break
    rule=ItemFinalPoseOverride.AnimGraphNode_StateMachine_0.Sitting FinalPose.AnimStateTransitionNode_1.Transition
    src=Sitting
    via=AnimStateTransitionNode_1
   }
  ]
 }
 Intro {
  path=ItemFinalPoseOverride.AnimGraphNode_StateMachine_0.Sitting FinalPose.AnimStateNode_0.Intro
  nodes [
   AnimGraphNode_ChooserPlayer_0 = Chooser Player: CHT_IntroSit!() -> (Pose:Pose Link Structure) {
    is=title AnimGraphNode_ChooserPlayer
    chooser {_structType=/Script/Chooser.EvaluateChooser, chooser=/WildEstate/Gameplay/VoidSlide/Sequence/ChooserTables/CHT_IntroSit}
    chooserContextDefinition [
     {
      _structType=/Script/Chooser.ContextObjectTypeClass
      class=&/VoidSlide_Player_AnimBP.VoidSlide_Player_AnimBP_C
      direction=ReadWrite
     }
     {struct=/Script/Chooser.ChooserPlayerSettings, _structType=/Script/Chooser.ContextObjectTypeStruct, direction=Read}
    ]
    defaultSettings {blendTime=0}
   }
   Output Animation Pose(Result=Modify Curve(SourcePose=@AnimGraphNode_ChooserPlayer_0, CurveValues_0=1, Alpha=1) {curveNames [SpawnProp], curveValues [0]}) {
    node {stateExitFunction {functionName=ANF_Disable Sit Additive}}
    stateExitFunction {bSelfContext=true, memberGuid=#4, memberName=ANF_Disable Sit Additive}
   }
  ]
 }
 AnimationBlendStackGraph_0 {
  path=ItemFinalPoseOverride.AnimGraphNode_StateMachine_0.Sitting FinalPose.AnimStateNode_0.Intro.AnimGraphNode_ChooserPlayer_0.AnimationBlendStackGraph_0
  nodes [Output Pose!(Result:Pose Link Structure=Blend Stack Input!() -> (Pose:Pose Link Structure) {is=title AnimGraphNode_BlendStackInput}) {is=title AnimGraphNode_BlendStackResult}]
 }
 Sit Idle Break {
  path=ItemFinalPoseOverride.AnimGraphNode_StateMachine_0.Sitting FinalPose.AnimStateNode_1.Sit Idle Break
  nodes [
   AnimGraphNode_ChooserPlayer_1 = Chooser Player: CHT_SitIdleBreak!() -> (Pose:Pose Link Structure) {
    is=title AnimGraphNode_ChooserPlayer
    chooser {_structType=&EvaluateChooser, chooser=/WildEstate/Gameplay/VoidSlide/Sequence/ChooserTables/CHT_SitIdleBreak}
    chooserContextDefinition [
     {
      _structType=&ContextObjectTypeClass
      class=&VoidSlide_Player_AnimBP_C
      direction=ReadWrite
     }
     {struct=&ChooserPlayerSettings, _structType=&ContextObjectTypeStruct, direction=Read}
    ]
   }
   Output Animation Pose(Result=Modify Curve(SourcePose=@AnimGraphNode_ChooserPlayer_1, CurveValues_0=1, Alpha=1) {curveNames [SpawnProp], curveValues [0]}) {
    node {stateEntryFunction {functionName=ANF_Disable Sit Additive}}
    stateEntryFunction {bSelfContext=true, memberGuid=#4, memberName=ANF_Disable Sit Additive}
   }
  ]
 }
 AnimationBlendStackGraph_0 {
  path=ItemFinalPoseOverride.AnimGraphNode_StateMachine_0.Sitting FinalPose.AnimStateNode_1.Sit Idle Break.AnimGraphNode_ChooserPlayer_1.AnimationBlendStackGraph_0
  nodes [Output Pose(Result=Blend Stack Input())]
 }
 Sitting {
  path=ItemFinalPoseOverride.AnimGraphNode_StateMachine_0.Sitting FinalPose.AnimStateNode_5.Sitting
  nodes [
   n1 = Blend Poses by int(
    BlendPose_0="Sequence Player (additive) 'WildEstate_VoidSlide_SeatSwap1'"() {bLoopAnimation=false, sequence=&/WildEstate_VoidSlide_SeatSwap1}
    BlendPose_1="Sequence Player (additive) 'WildEstate_VoidSlide_SeatSwap2'"() {bLoopAnimation=false, sequence=&/WildEstate_VoidSlide_SeatSwap2}
    BlendPose_2="Sequence Player (additive) 'WildEstate_VoidSlide_SeatSwap3'"() {bLoopAnimation=false, sequence=&/WildEstate_VoidSlide_SeatSwap3}
    BlendPose_3="Sequence Player (additive) 'WildEstate_VoidSlide_SeatSwap4'"() {bLoopAnimation=false, sequence=&/WildEstate_VoidSlide_SeatSwap4}
    BlendTime_0=0
    BlendTime_1=0
    BlendTime_2=0
    BlendTime_3=0
    ActiveChildIndex=Get SeatIndex() -> (SeatIndex=0)
   ) {
    blendPose [{linkId=-1, sourceLinkId=-1}, {linkId=-1, sourceLinkId=-1}, {linkId=-1, sourceLinkId=-1}, {linkId=-1, sourceLinkId=-1}]
    blendTime [0.1,0.1,0.1,0.1]
   }
   AnimGraphNode_ChooserPlayer_3 = Chooser Player: CHT_SitPoses!() -> (Pose:Pose Link Structure) {
    is=title AnimGraphNode_ChooserPlayer
    chooser {_structType=&EvaluateChooser, chooser=/WildEstate/Gameplay/VoidSlide/Sequence/ChooserTables/CHT_SitPoses}
    chooserContextDefinition [
     {
      _structType=&ContextObjectTypeClass
      class=&VoidSlide_Player_AnimBP_C
      direction=ReadWrite
     }
     {struct=&ChooserPlayerSettings, _structType=&ContextObjectTypeStruct, direction=Read}
    ]
    defaultSettings {bForceLooping=true, blendTime=0}
   }
   n2 = Apply Additive!(Base=@AnimGraphNode_ChooserPlayer_3, Additive=@n1, Alpha:float) -> (Pose)
   AnimGraphNode_ChooserPlayer_1 = Chooser Player: CHT_SitPoses() {
    chooser {_structType=&EvaluateChooser, chooser=&CHT_SitPoses}
    chooserContextDefinition [
     {
      _structType=&ContextObjectTypeClass
      class=&VoidSlide_Player_AnimBP_C
      direction=ReadWrite
     }
     {struct=&ChooserPlayerSettings, _structType=&ContextObjectTypeStruct, direction=Read}
    ]
    defaultSettings {bForceLooping=true, blendTime=0}
   }
   n3 = Get Sit Additive() -> (Sit Additive:bool)
   Output Animation Pose(Result=Modify Curve(
     SourcePose=Blend Poses by bool(BlendPose_0=@n2, BlendPose_1=@AnimGraphNode_ChooserPlayer_1, BlendTime_0=0, BlendTime_1=0.7, bActiveValue=@n3)
     CurveValues_0=1
     Alpha=1
    ) {curveNames [SpawnProp], curveValues [0]}) {node {updateFunction {functionName=OnUpdate_TriggerTime}}, updateFunction {bSelfContext=true, memberName=OnUpdate_TriggerTime}}
  ]
 }
 AnimationBlendStackGraph_0 {
  path=ItemFinalPoseOverride.AnimGraphNode_StateMachine_0.Sitting FinalPose.AnimStateNode_5.Sitting.AnimGraphNode_ChooserPlayer_1.AnimationBlendStackGraph_0
  nodes [Output Pose(Result=Blend Stack Input())]
 }
 AnimationBlendStackGraph_0 {
  path=ItemFinalPoseOverride.AnimGraphNode_StateMachine_0.Sitting FinalPose.AnimStateNode_5.Sitting.AnimGraphNode_ChooserPlayer_3.AnimationBlendStackGraph_0
  nodes [Output Pose(Result=Blend Stack Input())]
 }
 Transition {path=ItemFinalPoseOverride.AnimGraphNode_StateMachine_0.Sitting FinalPose.AnimStateTransitionNode_0.Transition, nodes [Result()]}
 Transition {
  path=ItemFinalPoseOverride.AnimGraphNode_StateMachine_0.Sitting FinalPose.AnimStateTransitionNode_1.Transition
  nodes [Result(bCanEnterTransition=Get bTriggerIdleBreak() -> (bTriggerIdleBreak:bool))]
 }
 Transition {path=ItemFinalPoseOverride.AnimGraphNode_StateMachine_0.Sitting FinalPose.AnimStateTransitionNode_2.Transition, nodes [Result()]}
 Transition {
  path=ItemFinalPoseOverride.AnimGraphNode_StateMachine_0.Sitting FinalPose.AnimStateTransitionNode_9.Transition
  nodes [Result(bCanEnterTransition=Get bSeatSwapped() -> (bSeatSwapped:bool))]
 }
 Final Pose Blend Out SM {
  entry=FinalPose For Sitting
  path=ItemFinalPoseOverride.AnimGraphNode_StateMachine_2.Final Pose Blend Out SM
  ids=id
  nodes [
   AnimStateNode_0 = FinalPose For Sitting() {is=title AnimStateNode}
   AnimStateNode_1 = FullBody Passthrough() {is=title AnimStateNode}
   AnimStateTransitionNode_1 = AnimStateTransitionNode() {bAllowInertializationForSelfTransitions=true}
  ]
  links [{
    dst=FullBody Passthrough
    rule=ItemFinalPoseOverride.AnimGraphNode_StateMachine_2.Final Pose Blend Out SM.AnimStateTransitionNode_1.Transition
    src=FinalPose For Sitting
    via=AnimStateTransitionNode_1
   }]
 }
 FinalPose For Sitting {
  path=ItemFinalPoseOverride.AnimGraphNode_StateMachine_2.Final Pose Blend Out SM.AnimStateNode_0.FinalPose For Sitting
  nodes [Output Animation Pose(Result=Two Way Blend(A=Use cached pose 'FinalPose PostOffsets'(), B=Use cached pose 'Final InputPose'(), Alpha=0) {blendNode {bAlwaysUpdateChildren=true}}) {
    node {updateFunction {functionName=OnUpdate}}
    updateFunction {bSelfContext=true, memberGuid=#5, memberName=OU_FinalPoseBlendOutTriggerTimer}
   }]
 }
 FullBody Passthrough {
  path=ItemFinalPoseOverride.AnimGraphNode_StateMachine_2.Final Pose Blend Out SM.AnimStateNode_1.FullBody Passthrough
  nodes [Output Animation Pose(Result=Use cached pose 'Final InputPose'())]
 }
 Transition {
  path=ItemFinalPoseOverride.AnimGraphNode_StateMachine_2.Final Pose Blend Out SM.AnimStateTransitionNode_1.Transition
  nodes [Result(bCanEnterTransition=Get IsTriggerFinalPoseBlendOut() -> (IsTriggerFinalPoseBlendOut:bool))]
 }
 ItemFullBodyAimOffsets {nodes [
   Input Pose!() -> (Pose, FullBody_AimOffsetAlpha:double, Aim_Yaw:double, Aim_Pitch:double) {name=InPoseFullBodyAimOffsets}
   Input Pose!() -> (Pose) {name=InPosePreFullBodyAimOffsets}
   Output Pose() {is=title AnimGraphNode_Root, layerGroup=Item Override}
  ]}
 ItemFullBodyOverride {nodes [
   Input Pose() {name=InPoseFullBody}
   n1 = Local To Component(LocalPose=Use cached pose 'PostSpawnProp Pose'())
   n2 = Local To Component(LocalPose=Use cached pose 'PostSpawnProp Pose'())
   n3 = Local To Component(LocalPose=Use cached pose 'PostSpawnProp Pose'())
   n4 = Local To Component(LocalPose=Use cached pose 'PostSpawnProp Pose'())
   n5 = Blend Poses by int(
    BlendPose_0=Component To Local(ComponentPose="Transform (Modify) Bone - Bone: Root"(ComponentPose=@n1, bAlphaBoolEnabled=True, Translation="7,-28,0") {alphaInputType=Bool, translationMode=BMM_Additive, boneToModify {boneName=Root}, translation=<7 -28>})
    BlendPose_1=Component To Local(ComponentPose="Transform (Modify) Bone - Bone: Root"(ComponentPose=@n2, bAlphaBoolEnabled=True, Translation="5,-28,0") {alphaInputType=Bool, translationMode=BMM_Additive, boneToModify {boneName=Root}, translation=<5 -28>})
    BlendPose_2=Component To Local(ComponentPose="Transform (Modify) Bone - Bone: Root"(ComponentPose=@n3, bAlphaBoolEnabled=True, Translation="0,-28,0") {alphaInputType=Bool, translationMode=BMM_Additive, boneToModify {boneName=Root}, translation {y=-28}})
    BlendPose_3=Component To Local(ComponentPose="Transform (Modify) Bone - Bone: Root"(ComponentPose=@n4, bAlphaBoolEnabled=True, Translation="-8,-37,0") {alphaInputType=Bool, translationMode=BMM_Additive, boneToModify {boneName=Root}, translation=<-8 -37>})
    BlendTime_0=0
    BlendTime_1=0
    BlendTime_2=0
    BlendTime_3=0
    ActiveChildIndex=Get SeatIndex() -> (SeatIndex=0)
   ) {
    blendPose [{linkId=-1, sourceLinkId=-1}, {linkId=-1, sourceLinkId=-1}, {linkId=-1, sourceLinkId=-1}, {linkId=-1, sourceLinkId=-1}]
    blendTime [0.1,0.1,0.1,0.1]
   }
   Output Pose(Result=@n5) {layerGroup=Item Override}
   AnimGraphNode_StateMachine_0 = Sitting FullBody() {is=title AnimGraphNode_StateMachine}
   n6 = Apply Additive!(
    Base=@AnimGraphNode_StateMachine_0
    Additive="Sequence Player (additive) 'WildEstate_VoidSlide_SpawnProp1'"() {sequence=&WildEstate_VoidSlide_SpawnProp1}
    bAlphaBoolEnabled:bool=Get bIsSeat01SpawnProp()
   ) -> (Pose) {alphaInputType=Bool}
   n7 = "Sequence Player (additive) 'WildEstate_VoidSlide_SpawnProp3'"() {sequence=&WildEstate_VoidSlide_SpawnProp3}
   n8 = Get bIsSeat03SpawnProp()
   Save cached pose 'PostSpawnProp Pose'(Pose=Apply Additive(
     Base=Apply Additive(Base=@n6, Additive=@n7, bAlphaBoolEnabled=@n8) {alphaInputType=Bool}
     Additive="Sequence Player (additive) 'WildEstate_VoidSlide_SpawnProp4'"() {sequence=&WildEstate_VoidSlide_SpawnProp4}
     bAlphaBoolEnabled=Get bIsSeat04SpawnProp()
    ) {alphaInputType=Bool}) {cacheName=PostSpawnProp Pose}
  ]}
 Sitting FullBody {
  entry=Intro
  path=ItemFullBodyOverride.AnimGraphNode_StateMachine_0.Sitting FullBody
  ids=id
  nodes [
   Re-Enter Sitting() {also {aliases [Sit Idle Break, Sitting]}}
   AnimStateNode_0 = Intro()
   AnimStateNode_1 = Sit Idle Break()
   AnimStateNode_5 = Sitting() {bAlwaysResetOnEntry=true}
   AnimStateTransitionNode_0 = AnimStateTransitionNode() {
    bAllowInertializationForSelfTransitions=true
    bAutomaticRuleBasedOnSequencePlayerInState=true
    transitionStart {guid=#1, notifyName=superbunny}
   }
   AnimStateTransitionNode_1 = AnimStateTransitionNode() {bAllowInertializationForSelfTransitions=true, crossfadeDuration=0.7}
   AnimStateTransitionNode_2 = AnimStateTransitionNode() {
    bAllowInertializationForSelfTransitions=true
    bAutomaticRuleBasedOnSequencePlayerInState=true
    crossfadeDuration=0.5
    transitionStart {guid=#2, notifyName=Reset Idle Break Timer}
   }
   AnimStateTransitionNode_9 = AnimStateTransitionNode() {
    bAllowInertializationForSelfTransitions=true
    crossfadeDuration=0
    transitionStart {guid=#3, notifyName=Reenter Sitting Event}
   }
  ]
  links [
   {
    dst=Sitting
    rule=ItemFullBodyOverride.AnimGraphNode_StateMachine_0.Sitting FullBody.AnimStateTransitionNode_0.Transition
    src=Intro
    via=AnimStateTransitionNode_0
   }
   {
    dst=Sitting
    rule=ItemFullBodyOverride.AnimGraphNode_StateMachine_0.Sitting FullBody.AnimStateTransitionNode_9.Transition
    src=Re-Enter Sitting
    via=AnimStateTransitionNode_9
   }
   {
    dst=Sitting
    rule=ItemFullBodyOverride.AnimGraphNode_StateMachine_0.Sitting FullBody.AnimStateTransitionNode_2.Transition
    src=Sit Idle Break
    via=AnimStateTransitionNode_2
   }
   {
    dst=Sit Idle Break
    rule=ItemFullBodyOverride.AnimGraphNode_StateMachine_0.Sitting FullBody.AnimStateTransitionNode_1.Transition
    src=Sitting
    via=AnimStateTransitionNode_1
   }
  ]
 }
 Intro {
  path=ItemFullBodyOverride.AnimGraphNode_StateMachine_0.Sitting FullBody.AnimStateNode_0.Intro
  nodes [
   AnimGraphNode_ChooserPlayer_0 = Chooser Player: CHT_IntroSit() {
    chooser {_structType=&EvaluateChooser, chooser=&CHT_IntroSit}
    chooserContextDefinition [
     {
      _structType=&ContextObjectTypeClass
      class=&VoidSlide_Player_AnimBP_C
      direction=ReadWrite
     }
     {struct=&ChooserPlayerSettings, _structType=&ContextObjectTypeStruct, direction=Read}
    ]
    defaultSettings {blendTime=0}
   }
   Output Animation Pose(Result=Modify Curve(SourcePose=@AnimGraphNode_ChooserPlayer_0, CurveValues_0=1, Alpha=1) {curveNames [SpawnProp], curveValues [0]}) {
    node {stateFullyBlendedInFunction {functionName=ANF_Disable Sit Additive}}
    stateFullyBlendedInFunction {bSelfContext=true, memberGuid=#4, memberName=ANF_Disable Sit Additive}
   }
  ]
 }
 AnimationBlendStackGraph_0 {
  path=ItemFullBodyOverride.AnimGraphNode_StateMachine_0.Sitting FullBody.AnimStateNode_0.Intro.AnimGraphNode_ChooserPlayer_0.AnimationBlendStackGraph_0
  nodes [Output Pose(Result=Blend Stack Input()) {is=title AnimGraphNode_BlendStackResult}]
 }
 Sit Idle Break {
  path=ItemFullBodyOverride.AnimGraphNode_StateMachine_0.Sitting FullBody.AnimStateNode_1.Sit Idle Break
  nodes [
   AnimGraphNode_ChooserPlayer_1 = Chooser Player: CHT_SitIdleBreak() {
    chooser {_structType=&EvaluateChooser, chooser=&CHT_SitIdleBreak}
    chooserContextDefinition [
     {
      _structType=&ContextObjectTypeClass
      class=&VoidSlide_Player_AnimBP_C
      direction=ReadWrite
     }
     {struct=&ChooserPlayerSettings, _structType=&ContextObjectTypeStruct, direction=Read}
    ]
   }
   Output Animation Pose(Result=Modify Curve(SourcePose=@AnimGraphNode_ChooserPlayer_1, CurveValues_0=1, Alpha=1) {curveNames [SpawnProp], curveValues [0]}) {
    node {stateEntryFunction {functionName=ANF_Disable Sit Additive}}
    stateEntryFunction {bSelfContext=true, memberGuid=#4, memberName=ANF_Disable Sit Additive}
   }
  ]
 }
 AnimationBlendStackGraph_0 {
  path=ItemFullBodyOverride.AnimGraphNode_StateMachine_0.Sitting FullBody.AnimStateNode_1.Sit Idle Break.AnimGraphNode_ChooserPlayer_1.AnimationBlendStackGraph_0
  nodes [Output Pose(Result=Blend Stack Input())]
 }
 Sitting {
  path=ItemFullBodyOverride.AnimGraphNode_StateMachine_0.Sitting FullBody.AnimStateNode_5.Sitting
  nodes [
   n1 = Blend Poses by int(
    BlendPose_0="Sequence Player (additive) 'WildEstate_VoidSlide_SeatSwap1'"() {bLoopAnimation=false, sequence=&WildEstate_VoidSlide_SeatSwap1}
    BlendPose_1="Sequence Player (additive) 'WildEstate_VoidSlide_SeatSwap2'"() {bLoopAnimation=false, sequence=&WildEstate_VoidSlide_SeatSwap2}
    BlendPose_2="Sequence Player (additive) 'WildEstate_VoidSlide_SeatSwap3'"() {bLoopAnimation=false, sequence=&WildEstate_VoidSlide_SeatSwap3}
    BlendPose_3="Sequence Player (additive) 'WildEstate_VoidSlide_SeatSwap4'"() {bLoopAnimation=false, sequence=&WildEstate_VoidSlide_SeatSwap4}
    BlendTime_0=0
    BlendTime_1=0
    BlendTime_2=0
    BlendTime_3=0
    ActiveChildIndex=Get SeatIndex() -> (SeatIndex=0)
   ) {
    blendPose [{linkId=-1, sourceLinkId=-1}, {linkId=-1, sourceLinkId=-1}, {linkId=-1, sourceLinkId=-1}, {linkId=-1, sourceLinkId=-1}]
    blendTime [0.1,0.1,0.1,0.1]
   }
   AnimGraphNode_ChooserPlayer_3 = Chooser Player: CHT_SitPoses() {
    chooser {_structType=&EvaluateChooser, chooser=&CHT_SitPoses}
    chooserContextDefinition [
     {
      _structType=&ContextObjectTypeClass
      class=&VoidSlide_Player_AnimBP_C
      direction=ReadWrite
     }
     {struct=&ChooserPlayerSettings, _structType=&ContextObjectTypeStruct, direction=Read}
    ]
    defaultSettings {bForceLooping=true, blendTime=0}
   }
   n2 = Apply Additive!(Base=@AnimGraphNode_ChooserPlayer_3, Additive=@n1, Alpha:float) -> (Pose)
   AnimGraphNode_ChooserPlayer_1 = Chooser Player: CHT_SitPoses() {
    chooser {_structType=&EvaluateChooser, chooser=&CHT_SitPoses}
    chooserContextDefinition [
     {
      _structType=&ContextObjectTypeClass
      class=&VoidSlide_Player_AnimBP_C
      direction=ReadWrite
     }
     {struct=&ChooserPlayerSettings, _structType=&ContextObjectTypeStruct, direction=Read}
    ]
    defaultSettings {bForceLooping=true, blendTime=0}
   }
   n3 = Get Sit Additive()
   Output Animation Pose(Result=Modify Curve(
     SourcePose=Blend Poses by bool(BlendPose_0=@n2, BlendPose_1=@AnimGraphNode_ChooserPlayer_1, BlendTime_0=0, BlendTime_1=0.7, bActiveValue=@n3)
     CurveValues_0=1
     Alpha=1
    ) {curveNames [SpawnProp], curveValues [0]}) {node {updateFunction {functionName=OnUpdate_TriggerTime}}, updateFunction {bSelfContext=true, memberName=OnUpdate_TriggerTime}}
  ]
 }
 AnimationBlendStackGraph_0 {
  path=ItemFullBodyOverride.AnimGraphNode_StateMachine_0.Sitting FullBody.AnimStateNode_5.Sitting.AnimGraphNode_ChooserPlayer_1.AnimationBlendStackGraph_0
  nodes [Output Pose(Result=Blend Stack Input())]
 }
 AnimationBlendStackGraph_0 {
  path=ItemFullBodyOverride.AnimGraphNode_StateMachine_0.Sitting FullBody.AnimStateNode_5.Sitting.AnimGraphNode_ChooserPlayer_3.AnimationBlendStackGraph_0
  nodes [Output Pose(Result=Blend Stack Input())]
 }
 Transition {path=ItemFullBodyOverride.AnimGraphNode_StateMachine_0.Sitting FullBody.AnimStateTransitionNode_0.Transition, nodes [Result()]}
 Transition {
  path=ItemFullBodyOverride.AnimGraphNode_StateMachine_0.Sitting FullBody.AnimStateTransitionNode_1.Transition
  nodes [Result(bCanEnterTransition=Get bTriggerIdleBreak())]
 }
 Transition {path=ItemFullBodyOverride.AnimGraphNode_StateMachine_0.Sitting FullBody.AnimStateTransitionNode_2.Transition, nodes [Result()]}
 Transition {
  path=ItemFullBodyOverride.AnimGraphNode_StateMachine_0.Sitting FullBody.AnimStateTransitionNode_9.Transition
  nodes [Result(bCanEnterTransition=Get bSeatSwapped())]
 }
 ItemIdleOverride {nodes [
   Input Pose() {name=InPose_Idle}
   Output Pose() {is=title AnimGraphNode_Root, layerGroup=Item Override}
  ]}
 ItemJetPackJumpAdditive {nodes [Output Pose() {layerGroup=Item Override}]}
 ItemJetPackStartAdditive {nodes [Output Pose() {layerGroup=Item Override}]}
 ItemJumpLoopAdditive {nodes [
   Input Pose() {name=InPoseItemJumpLoopAdditive}
   Output Pose() {layerGroup=Item Override}
  ]}
 ItemJumpOverride {nodes [
   Input Pose() {name=InPoseJump}
   Output Pose() {layerGroup=Item Override}
  ]}
 ItemJumpUpAdditive {nodes [
   Input Pose() {name=InPoseItemJumpUpAdditive}
   Output Pose() {layerGroup=Item Override}
  ]}
 ItemPreAimOffsets {nodes [
   Input Pose() {name=InPosePreAimOffsets}
   Output Pose() {layerGroup=Item Override}
  ]}
 ItemPreIK {nodes [
   Input Pose() {name=InPose_PreIK}
   Output Pose() {layerGroup=Item Override}
  ]}
 ItemRightHandIK {nodes [
   Input Pose() {name=InPosePreRightHandIK}
   Input Pose() {name=InPosePostRightHandIK}
   Output Pose() {layerGroup=Item Override}
  ]}
 ItemSkeletalControl {nodes [
   Input Pose()
   Output Pose() {layerGroup=Item Override}
  ]}
 ItemSwimJumpFallLoop {nodes [Output Pose() {layerGroup=Item Override}]}
 ItemSwimJumpSurfaceEnd {nodes [Output Pose() {layerGroup=Item Override}]}
 ItemSwimJumpSurfaceLoop {nodes [Output Pose() {layerGroup=Item Override}]}
 ItemUpperBody {nodes [
   Input Pose!() -> (Pose, InputParam:Fort Anim Input Adjusted Aim Structure) {name=InPoseUpperBody}
   Output Pose() {layerGroup=Item Override}
  ]}
 ItemZipLineStartAdditive {nodes [Output Pose() {layerGroup=Item Override}]}
 OU_FinalPoseBlendOutTriggerTimer {nodes [
   n1 = OU_FinalPoseBlendOutTriggerTimer!() -> (then, Context:"Anim Update Context Structure (by ref)", Node:"Anim Node Reference Structure (by ref)") {is=title K2Node_FunctionEntry}
   > n2 = Set FinalPoseBlendOutTriggerTime(FinalPoseBlendOutTriggerTime:double=float - float(
     A=Get FinalPoseBlendOutTriggerTime() -> (FinalPoseBlendOutTriggerTime:double)
     B=GetDeltaTime!(Context:"Anim Update Context Structure (by ref)"=@n1) -> (ReturnValue:float)
    )) -> (Output_Get:double)
   > Branch(Condition="float <= float"(A=@n2) -> (ReturnValue=~))
   > Set IsTriggerFinalPoseBlendOut(IsTriggerFinalPoseBlendOut:bool=true) -> (Output_Get:bool)
   > Return Node()
  ]}
]
variables [
 FinalPoseBlendOutTriggerTime {type=number, default=0.05}
 IsTriggerFinalPoseBlendOut {type=boolean}
 Max Idle Break Time {type=number, default=10}
 Min Idle Break Time {type=number, default=5}
 Sit Additive {type=boolean}
]
